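# Physics asset for the Turret_Gun gadget, giving collision volumes to a seated, twin-gun turret.
# Hand-fitted boxes and capsules on each bone cover the base, chair, pedals, arms and two-part shield.
# Each gun body combines a capsule barrel with boxes for the body and angled ammo container.
# The asset has no constraints, so no physics joints link the bodies.
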PhysicsAsset /Game/Gadgets/Assets/Turret_Gun/Meshes/Turret_Gun_Physics
bodies {
 chair_arm [{shapeType=Box, extentX=80, extentY=18, extentZ=18, shapeName=LowerArm, center=<25.6305 -0.0005 12.2495>, rotation=<rot 34.9998 0 0>}]
 chair_back [{shapeType=Box, extentX=6, extentY=31, extentZ=25, shapeName=ChairBack, center=<-11.883 0.0004 24.6853>}]
 chair_base [{shapeType=Box, extentX=73, extentY=35.5, extentZ=18.5, center=<-13.1456 0 -1.3085>}]
 gun_arm_base [
  {shapeType=Box, extentX=19.5, extentY=78.5, extentZ=30.9375, shapeName=ShieldTop, center=<22.8142 0 23.2952>}
  {shapeType=Box, extentX=20.5, extentY=55, extentZ=37.5, shapeName=ShieldBottom, center=<23.0136 0.7895 -10.1395>}
 ]
 gun_arm_bott [
  {shapeType=Box, extentX=32, extentY=28.5, extentZ=19.5, shapeName=MidGun, center=<-4.621 -0.0003 44.384>}
  {shapeType=Box, extentX=13.5, extentY=13, extentZ=33, shapeName=MidArm, center=<16.6872 0 32.4494>, rotation=<rot 44.9997 0 0>}
 ]
 gun_body_l [
  {shapeType=Capsule, length=115.5, radius=7, shapeName=Barrel_Left, center=<84.0669 -0.0011 10.3244>, rotation=<rot 90 0 0>}
  {shapeType=Box, extentX=28, extentY=39.5, extentZ=27, shapeName=Ammo_Left, center=<3.1829 -21.94 -8.0268>, rotation=<rot 0 0 -39.9955>}
  {shapeType=Box, extentX=46.5, extentY=21, extentZ=32, shapeName=GunBody_Left, center=<2 -0.4397 4.1894>}
 ]
 gun_body_r [
  {shapeType=Capsule, length=116.5, radius=7.25, center=<-82.0184 0.0002 -10.1252>, rotation=<rot -90 0.02 -0.0202>}
  {shapeType=Box, extentX=46, extentY=21, extentZ=31, shapeName=GunBody_Right, center=<-0.9998 0.497 -4.7026>}
  {shapeType=Box, extentX=29.5, extentY=38, extentZ=25.5, shapeName=Ammo_Right, center=<-3.2521 21.2243 7.0537>, rotation=<rot 0 0 -39.9962>}
 ]
 pedal_l [{shapeType=Box, extentX=4, extentY=15.5, extentZ=15, shapeName=Pedal_Left, center=<-0.3075 -5.8807 0.053>, rotation=<rot -29.9998 0 0>}]
 pedal_r [{shapeType=Box, extentX=4.5, extentY=15.5, extentZ=14, shapeName=Pedal_Right, center=<0.6471 5.6489 -0.5745>, rotation=<rot -24.9999 2 0>}]
 root [
  {shapeType=Capsule, length=12.5, radius=33.25, shapeName=BaseCapsule, center=<0 0 58.7666>, rotation=<rot 0 0 -0.0009>}
  {shapeType=Box, extentX=136.5, extentY=142.5, extentZ=12.5, shapeName=UpperBase, center=<0 0 33.4008>}
  {shapeType=Box, extentX=147.5, extentY=187.5, extentZ=15, center=<0.2444 0 20.8995>}
 ]
}
constraints []
